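; Parent movement mode extension for the VanillaSoda weapon's guided motion that ends the movement when the owning pawn is about to be launched.
; On activation it caches its runtime data and starts an async wait for the Event.Character.PreLaunch gameplay event on the pawn.
; Receiving that event calls ForceManualEnd on the cached runtime data.
; Deactivation ends the wait action.
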(event EventOnExtensionActivate
  (params (RuntimeContext :type MMERuntime%20Context%20Structure%20%28by%20ref%29))
  (bind _returnvalue (Class|FortMovementModeExtBlueprintLibrary|GetRuntimeDatafromContext RuntimeContext "/Script/FortniteGame.FortMovementMode_BaseExtRuntimeData"))
  (Variables|Default|SetRuntimeData _returnvalue)
  (bind _asynctaskproxy (Ability|Async|WaitGameplayEventToActor (Variables|FortMovementMode_BaseExtLogic|GetOwnerPawn) "(TagName=\"Event.Character.PreLaunch\")")
    (:then
      (Variables|Default|SetWaitGameplayEventAction _asynctaskproxy))
    (:EventReceived
      (Class|FortMovementModeBaseExtRuntimeData|ForceManualEnd (Variables|Default|GetRuntimeData)))))

(event EventOnExtensionDeactivate
  (params (RuntimeContext :type MMERuntime%20Context%20Structure%20%28by%20ref%29))
  (Ability|Async|EndAction (Variables|Default|GetWaitGameplayEventAction)))
